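# Niagara system for the broken bridge waterfall effect in the Chapter 6 BR map, named for splash and wisp elements.
# It holds three emitters, Wisp, SplashFlat and Splash, with no renderers or emitter-level modules listed.
# SystemState is the only module recorded, and a user position parameter, OffsetPosition, is exposed with no default value.
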NiagaraSystem /BRMapCh6/FX/NSLW_BrokenBrige_Waterfall
dependencies {usedDataInterfaces [], usedDynamicInputs [], usedModules [&SystemState]}
emitters {Splash {rendererClasses [], renderers []}, SplashFlat {rendererClasses [], renderers []}, Wisp {rendererClasses [], renderers []}}
summary {
 systemName=NSLW_BrokenBrige_Waterfall
 emitters [{emitterName=Wisp, rendererClasses []}, {emitterName=SplashFlat, rendererClasses []}, {emitterName=Splash, rendererClasses []}]
 userVariables [User.OffsetPosition {type=NiagaraPosition, description="", defaultValue {}}]
}
