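# Niagara system for the active state of a dance point of interest: a colored light pillar with pulsing emissive rings.
# LightPillar spawns one persistent mesh tinted by User.Color that scales up quickly, with sine-driven alpha biased by time of day.
# Ring_Emissive bursts two offset meshes every User.PulseTime, curve-animating scale, alpha and material parameters, with User.Brightness boosting RGB.
# Ring alpha follows time of day, and its X-scale pulse is remapped by LOD distance.
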NiagaraSystem /BRMapCh6/FX/NS_POI_Dance_Active
dependencies {
 usedDataInterfaces [&NiagaraDataInterfaceCurve, &NiagaraDataInterfaceVectorCurve]
 usedDynamicInputs []
 usedModules [&SystemState]
}
emitters {
 LightPillar {
  emitterUpdateScript [
   &EmitterState(
    Enable Distance Culling=true
    Life Cycle Mode=::NewEnumerator1
    Loop Behavior=::NewEnumerator1
    Loop Delay=0.15
    Loop Duration Mode=::NewEnumerator1
    Min Distance Response=::SleepAndLetParticlesFinish
    MinDistance=1000
    Module.bUseMaxDistance=false
    Module.bUseMinDistance=true
    Scalability Mode=::NewEnumerator1
    UseLoopDelay=false
   )
   &SpawnBurst_Instantaneous(Spawn Count=1, Use Spawn Probability=false)
  ]
  particleSpawnScript [
   &InitializeParticle(Color=$User.Color, Lifetime=2.5, Mesh Scale=<8 3 30>, Mesh Scale Mode=::NewEnumerator4, UsePositionOffset=false, Write Material Random=true)
   &InitialMeshOrientation(Orientation Vector=<0 0 1>, Use Orientation Vector=true, Use Rotation Vector=false)
   &ShapeLocation(MaskSpawnGroup=false, Offset=<0 0 -30>, Offset Mode=::NewEnumerator0, Override Seed=false, Sphere Radius=1, UseOwnerScale=true)
  ]
  particleUpdateScript [
   &ParticleState(Kill Particles When Lifetime Has Elapsed=false, Loop Particles Lifetime=true)
   &ScaleColor(
    Scale Alpha=&Sine(
     Normalized Angle=$Emitter.Age
     Period=3
     Scale=0.05
     Bias=&FNTimeOfDayScalar(Night Scale=0.175, Morning Scale=0.2, Day Scale=0.225, Evening Scale=0.2)
    )
    ScaleA=true
    ScaleRGB=false
   )
   &ScaleMeshSize(Scale Factor=&VectorFromCurve(
     VectorCurve=DataInterface() {propertyValues {
       XCurve {keys [{interpMode=RCIM_Cubic, tangentMode=RCTM_User, time=0, value=0}, {interpMode=RCIM_Cubic, tangentMode=RCTM_User, time=0.3, value=1}]}
       YCurve {keys [{interpMode=RCIM_Cubic, tangentMode=RCTM_User, time=0, value=0}, {interpMode=RCIM_Cubic, tangentMode=RCTM_User, time=0.3, value=1}]}
       ZCurve {keys [{interpMode=RCIM_Cubic, tangentMode=RCTM_User, time=0, value=0}, {interpMode=RCIM_Cubic, tangentMode=RCTM_User, time=0.3, value=1}]}
      }}
     CurveIndex=$Emitter.Age
    ))
   &SolveForcesAndVelocity(Clamp Velocity=false, Limit Acceleration=false)
  ]
  rendererClasses [&NiagaraMeshRendererProperties]
  renderers [{rendererClass=&NiagaraMeshRendererProperties, rendererIndex=0}]
 }
 Ring_Emissive {
  emitterUpdateScript [
   &EmitterState(Life Cycle Mode=::NewEnumerator1, Loop Duration=$User.PulseTime, UseLoopDelay=false)
   &SpawnBurst_Instantaneous(Spawn Count=2, Use Spawn Probability=false, UseLoopCountLimit=false)
  ]
  particleSpawnScript [
   &InitializeParticle(
    Lifetime=$User.PulseTime
    Mesh Scale=&MaskVector(
     Mask Source=1
     Mask Parameter=&ReturnExecIndex()
     Mask=0
     Masked Vector=<8 10 10>
     Passthrough Vector=<10 12 12>
    )
    Mesh Scale Mode=::NewEnumerator4
    Position Offset=&MaskVector(
     Mask Source=1
     Mask Parameter=&ReturnExecIndex()
     Mask=0
     Masked Vector=<0 0 -10>
     Passthrough Vector=<1 0 -35>
    )
    UsePositionOffset=true
    Write Material Random=true
   )
   &CameraOffset()
   &InitialMeshOrientation(Mesh Orientation Mode=::NewEnumerator2, Orientation Vector=<0 0 1>, Use Rotation Vector=false)
  ]
  particleUpdateScript [
   &ParticleState()
   &ScaleMeshSize(Scale Factor=&MakeVector(
     X=&FloatFromCurve(
      FloatCurve=DataInterface() {propertyValues {Curve {keys [
          {arriveTangent=1.3577, interpMode=RCIM_Cubic, leaveTangent=1.3577, tangentMode=RCTM_User, time=0, value=0.7}
          {interpMode=RCIM_Cubic, tangentMode=RCTM_User, time=0.4, value=1}
          {tangentMode=RCTM_User, time=1, value=0.7}
         ]}}}
      Scale Curve=&FN_LOD_Distance_Remap(Start Distance=1000, End Distance=2250, Minimum Output=0.5, Maximum Output=1, Clamp Output=true)
     )
     Y=1
     Z=1
    ))
   &ScaleColor(
    Scale Alpha=&FloatFromCurve(
     FloatCurve=DataInterface() {propertyValues {Curve {keys [
         {arriveTangent=0.0975, interpMode=RCIM_Cubic, leaveTangent=0.0975, tangentMode=RCTM_User, time=0, value=0.9}
         <at 0.3 1 RCIM_Cubic>
         {arriveTangent=0.0618, interpMode=RCIM_Cubic, leaveTangent=0.0618, tangentMode=RCTM_User, time=1, value=0.9}
        ]}}}
     Scale Curve=&FNTimeOfDayScalar(Night Scale=0.6, Morning Scale=0.7, Evening Scale=0.7)
    )
    Scale RGB=&VectorFromFloat(Value=$User.Brightness)
    ScaleA=true
    ScaleRGB=true
   )
   &DynamicMaterialParameters(
    Index 0 Param 1=&FloatFromCurve(FloatCurve=DataInterface() {propertyValues {Curve {keys [
         {interpMode=RCIM_Cubic, tangentMode=RCTM_User, time=0, value=0.2}
         {interpMode=RCIM_Cubic, tangentMode=RCTM_User, time=0.6, value=0}
         {interpMode=RCIM_Cubic, tangentMode=RCTM_User, time=1, value=0.2}
        ]}}})
    Index 0 Param 2=&FloatFromCurve(
     FloatCurve=DataInterface() {propertyValues {Curve {keys [
         {arriveTangent=2.2195, interpMode=RCIM_Cubic, leaveTangent=2.2195, tangentMode=RCTM_User, time=0, value=0.5}
         {interpMode=RCIM_Cubic, tangentMode=RCTM_User, time=0.4, value=1}
         {tangentMode=RCTM_User, time=1, value=0.5}
        ]}}}
     Scale Curve=50
    )
    Module.Param0WriteEnabled=true
    Module.Param1WriteEnabled=true
    Module.Param2WriteEnabled=true
    Module.Param3WriteEnabled=true
   )
  ]
  rendererClasses [&NiagaraMeshRendererProperties]
  renderers [{rendererClass=&NiagaraMeshRendererProperties, rendererIndex=0}]
 }
}
summary {
 systemName=NS_POI_Dance_Active
 emitters [
  {emitterName=Ring_Emissive, rendererClasses [&NiagaraMeshRendererProperties]}
  {emitterName=LightPillar, rendererClasses [&NiagaraMeshRendererProperties]}
 ]
 userVariables [
  User.Brightness {type=NiagaraFloat, defaultValue=3, description=""}
  User.Color {type=LinearColor, description="", defaultValue {struct=LinearColor, value=<rgba 0.7708 0.111 0.6894 1>}}
  User.IsDancing {type=NiagaraBool, defaultValue=true}
  User.PatternOffset {type=NiagaraFloat, defaultValue=75, description=""}
  User.PulseTime {type=NiagaraFloat, defaultValue=4, description=""}
 ]
}
